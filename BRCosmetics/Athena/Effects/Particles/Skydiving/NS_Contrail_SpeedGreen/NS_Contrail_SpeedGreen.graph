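# Skydiving contrail cosmetic that sheds orange-gold leaves and pale pink streaks from character sockets.
# FN_Contrail_Setup's dive value drives spawn rates, streak scale and front-end motion.
# Sprite and mesh leaves drift on curl noise or spiral via waveform-driven position offsets, while thin streak meshes accompany two persistent skeletal-attached trail meshes.
# Time-of-day inputs adjust brightness, and front-end switches retune motion, orientation and intensity for lobby preview.
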NiagaraSystem /BRCosmetics/Athena/Effects/Particles/Skydiving/NS_Contrail_SpeedGreen/NS_Contrail_SpeedGreen
dependencies {
 usedDataInterfaces [
  &NiagaraDataInterfaceArrayFloat
  &NiagaraDataInterfaceArrayColor
  &NiagaraDataInterfaceCurve
  &NiagaraDataInterfaceSpriteRendererInfo
  &NiagaraDataInterfaceArrayFloat2
  &NiagaraDataInterfaceMeshRendererInfo
  &NiagaraDataInterfaceArrayFloat3
  &NiagaraDataInterfaceVectorCurve
  &NiagaraDataInterfaceSocketReader
  &NiagaraDataInterfaceSkeletalMesh
 ]
 usedDynamicInputs [&SelectVector2DFromArray, &FNTimeOfDayColor]
 usedModules [
  &self:SystemSpawnScript.SystemScriptSource.SystemScriptGraph.NiagaraNodeAssignment_3.SetVariables_#1
  &self:SystemSpawnScript.SystemScriptSource.SystemScriptGraph.NiagaraNodeAssignment_2.SetVariables_#2
  &SystemState
  /Game/Effects/Niagara/Modules/Contrails/FN_Contrail_Setup
  &FN_Integrator
  &self:SystemSpawnScript.SystemScriptSource.SystemScriptGraph.NiagaraNodeAssignment_1.SetVariables_#3
 ]
}
emitters {
 Leaves01 {
  emitterUpdateScript [
   &EmitterState(Delay First Loop Only=true, Life Cycle Mode=::NewEnumerator1, Loop Delay=0.22, Loop Duration=0.25, UseLoopDelay=true)
   &SpawnRate(SpawnRate=&Lerp_Float(A=20, B=30, Alpha=$System.FN_Contrail_Setup.Dive Value), Use Spawn Probability=false) {module=SpawnRate001}
  ]
  particleSpawnScript [
   &PartitionParticles()
   &InitializeParticle(
    Color=&SelectLinearColorFromArray(
     Array Sampling Mode=::NewEnumerator0
     Direct Array Index=?
     Color Selection Array=DataInterface() {propertyValues {ColorData [<rgba 0.75 0.275 0.0375 1>, <rgba 0.7188 0.1702 0.0179 1>, <rgba 0.75 0.4712 0.024 1>, <rgba 0.8125 0.1198 0.0264 1>]}}
     Random Seed=0
    )
    Lifetime Max=1
    Lifetime Min=0.6
    Lifetime Mode=::NewEnumerator1
    Sprite Rotation Mode=::NewEnumerator2
    Sprite Size=&Multiply_Vector2D_ByFloat(
     Float=&RandomRangeFloat(Minimum=0.8, Maximum=1.3)
     Vector2D=&RandomRangeVector2D(Minimum=<7.5 2.5>, Maximum=<12.5 9>)
    )
    Sprite Size Mode=::NewEnumerator4
    UsePositionOffset=false
    Write Material Random=true
   ) {module=InitializeParticle001}
   &SocketLocation(
    Interpolate Socket Sampling=true
    SamplePosOffset=false
    Socket Index=$Particles.PartitionParticles.Partition
    Socket Reader=$System.Socket Reader
    Socket Sampling Mode=::NewEnumerator3
    VelocitySpeedLimit=false
   )
   &ShapeLocation(MaskSpawnGroup=false, Override Seed=false, Sphere Radius=4, UseOwnerScale=true)
   &DynamicMaterialParameters(
    Index 0 Param 1=&FN_FrontEndFloatModifier(
     Gameplay=&FNTimeOfDayScalar(Night Scale=0.06, Morning Scale=0.06, Day Scale=0.1, Evening Scale=0.08)
     Front End=0.01
    )
    Module.Param0WriteEnabled=true
    Module.Param1WriteEnabled=true
    Module.Param2WriteEnabled=true
    Module.Param3WriteEnabled=true
   ) {module=DynamicMaterialParameters001}
   &InitialMeshOrientation(Use Rotation Vector=false)
   &AddVelocity(Velocity=&FN_FrontEndVector3Modifier(
     Gameplay=&RandomRangeVector(Minimum=<0 0 400>, Maximum=<0 0 600>)
     Front End=&Lerp_Vector(
      A=&RandomRangeVector(Minimum=<0 0 200>, Maximum=<1 1 400>)
      B=<0 -175 175>
      Alpha=$System.FN_Contrail_Setup.Dive Value
     )
    )) {module=AddVelocity002}
   &CameraOffset(Camera Offset Amount=&RandomRangeFloat(Minimum=-10, Maximum=-15))
   &OffsetPosition(Position Offset=<0 0 10>)
  ]
  particleUpdateScript [
   &ParticleState()
   &ScaleSpriteSize(Uniform Curve Sprite Scale=DataInterface() {propertyValues {Curve {
       defaultValue=1
       keys [
        {arriveTangent=7.4671, interpMode=RCIM_Cubic, leaveTangent=7.4671, tangentMode=RCTM_User, time=0, value=0.5}
        {arriveTangent=0.0097, interpMode=RCIM_Cubic, leaveTangent=0.0097, tangentMode=RCTM_User, time=0.1, value=1}
        {interpMode=RCIM_Cubic, tangentMode=RCTM_User, time=0.5, value=1}
        {tangentMode=RCTM_User, time=1, value=0}
       ]
      }}})
   &UpdateMeshOrientation(
    Rotation Rate=&FloatFromCurve(
     FloatCurve=DataInterface() {propertyValues {Curve {
        defaultValue=1
        keys [{interpMode=RCIM_Cubic, tangentMode=RCTM_User, time=0, value=1}, {interpMode=RCIM_Cubic, tangentMode=RCTM_User, time=1, value=0}]
       }}}
     Scale Curve=&RandomRangeFloat(Minimum=0.5, Maximum=1)
    )
    Rotation Vector=&RandomVector(Vector Scale=1)
   )
   &SubUVAnimation(
    Random Start Frame=true
    "Sprite Renderer "=DataInterface() {propertyValues {SpriteRenderer=&self:Leaves01_0.NiagaraSpriteRendererProperties_2}}
    SubUV Lookup Index Scale=0.5
    UseEndFrame=true
    UseStartFrame=true
   ) {module=SubUVAnimation002}
   /Niagara/Modules/Update/Forces/CurlNoiseForce(
    Noise Frequency=100
    Noise Strength=&Multiply_Float(
     A=100
     B=&FloatFromCurve(FloatCurve=DataInterface() {propertyValues {Curve {keys [{interpMode=RCIM_Cubic, tangentMode=RCTM_User, time=0, value=0}, {interpMode=RCIM_Cubic, tangentMode=RCTM_User, time=1, value=1}]}}})
    )
    Pan Noise=false
   )
   &AerodynamicDrag(
    Aerodynamic Drag=0.01
    Aerodynamic Rotational Drag=0.5
    Drag Shape Curve=DataInterface() {propertyValues {Curve {keys [
        {interpMode=RCIM_Cubic, tangentMode=RCTM_User, time=0, value=0.1}
        {arriveTangent=2.0799, interpMode=RCIM_Cubic, leaveTangent=2.0799, tangentMode=RCTM_User, time=1, value=1}
       ]}}}
    Lift Shape Curve=DataInterface() {propertyValues {Curve {keys [
        {interpMode=RCIM_Cubic, tangentMode=RCTM_User, time=0, value=0}
        {interpMode=RCIM_Cubic, tangentMode=RCTM_User, time=0.5, value=1}
        {tangentMode=RCTM_User, time=1, value=0}
       ]}}}
    Use Linear Drag=true
    Use Rotational Drag=true
   )
   &SolveForcesAndVelocity(Clamp Velocity=false, Limit Acceleration=false)
   &AlignSpriteToMeshOrientation(SetAlignment=true, SetFacing=true)
   &ScaleColor(
    Scale RGB=&VectorFromFloat(Value=&FN_FrontEndFloatModifier(
      Gameplay=&FNTimeOfDayScalar(Night Scale=0.2, Morning Scale=0.25, Day Scale=0.5, Evening Scale=0.35)
      Front End=0.31
     ))
    ScaleA=true
    ScaleRGB=true
   )
  ]
  rendererClasses [&NiagaraSpriteRendererProperties]
  renderers [{rendererClass=&NiagaraSpriteRendererProperties}]
 }
 Leaves02_Hand {
  emitterUpdateScript [
   &EmitterState(
    Delay First Loop Only=true
    Inactive Response=::NewEnumerator1
    Life Cycle Mode=::NewEnumerator1
    Loop Delay=0.5
    Loop Duration=1
    UseLoopDelay=true
   )
   &SpawnRate(SpawnRate=&Lerp_Float(A=45, B=55, Alpha=$System.FN_Contrail_Setup.Dive Value), Use Spawn Probability=true)
  ]
  particleSpawnScript [
   &PartitionParticles(Write Partition to Ribbon ID=::NewEnumerator0)
   &InitializeParticle(
    Color=&SelectLinearColorFromArray(
     Array Sampling Mode=::NewEnumerator0
     Direct Array Index=?
     Color Selection Array=DataInterface() {propertyValues {ColorData [<rgba 0.75 0.275 0.0375 1>, <rgba 0.7188 0.1702 0.0179 1>, <rgba 0.75 0.4712 0.024 1>, <rgba 0.8125 0.1198 0.0264 1>]}}
     Random Seed=0
    )
    Lifetime=&RandomRangeFloat(Minimum=0.5, Maximum=1.5)
    Mesh Renderer Array Visibility Mode=::NewEnumerator1
    Mesh Renderer Info=DataInterface() {propertyValues {MeshRenderer=&self:Leaves02_Hand_0.NiagaraMeshRendererProperties_1}}
    Mesh Renderer Mesh Index:NiagaraInt32=/Niagara/DynamicInputs/Divide/Divide_Int(A:NiagaraInt32=&ReturnParticleIDAttributes(), B:NiagaraInt32=2)
    Mesh Scale=&Multiply_VectorByFloat(
     Vector=&RandomRangeVector(Minimum=<1 1 1>, Maximum=<0.5 0.5 0.5>)
     Float=&RandomRangeFloat(Minimum=0.05, Maximum=0.15)
    )
    Mesh Scale Mode=::NewEnumerator4
    UsePositionOffset=true
    Write Material Random=true
   )
   &InitialMeshOrientation(
    Mesh Axis To Orient=<0 0 0>
    Mesh Orientation Mode=::NewEnumerator2
    Orientation Vector=&SelectVectorFromArray(
     Array Sampling Mode=::NewEnumerator0
     Direct Array Index=?
     Vector Selection Array=DataInterface() {propertyValues {FloatData [<0 0 -1>, <0 0 1>]}}
     Random Seed=0
    )
    Rotation=&RandomRangeVector(Minimum=<0 0 -1>, Maximum=<0 0 1>)
    Use Rotation Vector=true
   )
   &CameraOffset(Camera Offset Amount=&FN_FrontEndFloatModifier(Gameplay=&RandomRangeFloat(Minimum=-15, Maximum=-20), Front End=-10)) {module=CameraOffset002}
  ]
  particleUpdateScript [
   &ParticleState()
   &ScaleMeshSize(Scale Factor=&VectorFromCurve(VectorCurve=DataInterface() {propertyValues {
       XCurve {keys [<at 0.09 0>,<at 0.1 1>,<at 1 1>]}
       YCurve {keys [<at 0 0>,<at 0.2 1>,<at 0.9 1>,<at 1 0>]}
       ZCurve {keys [<at 0.09 0>,<at 0.1 1>,<at 1 1>]}
      }})) {module=ScaleMeshSize001}
   &SocketLocation(
    Interpolate Socket Sampling=true
    Orientation Sampling=::NewEnumerator1
    SamplePosOffset=true
    Sampled Position Offset=&MergeHelper_VectorSwitch(0=<20 -10 5>, 1=<-20 10 -5>, 2=?, 3=?, Index=$Particles.PartitionParticles.Partition, Num=::NewEnumerator0)
    Socket Index=$Particles.PartitionParticles.Partition
    Socket Reader=$System.Socket Reader
    Socket Sampling Mode=::NewEnumerator3
    VelocitySpeedLimit=false
   )
   /SetVariables(Particles.CustomRandom:NiagaraFloat=&Add_Float(
     A=&RandomRangeFloat(
      Minimum=0
      Maximum=50
      Randomness Mode=::NewEnumerator1
      Random Seed=&Add_Integer(A=$Particles.PartitionParticles.Partition, B=0)
      Fixed Random Seed=true
     )
     B=&MergeHelper_FloatSwitch(0=0.7, 1=0.88, 2=?, 3=?, Index=$Particles.PartitionParticles.Partition, Num=::NewEnumerator0)
    )) {isSetParameters=true}
   /SetVariables(Particles.SpiralVector:Vector3f=&MakeVector(
     X=&Waveform(
      Global Input Position=$Particles.CustomRandom
      Global Amplitude Scale=&MergeHelper_FloatSwitch(
       0=&FloatFromCurve(
        FloatCurve=DataInterface() {propertyValues {Curve {keys [
            {arriveTangent=0.1886, interpMode=RCIM_Cubic, leaveTangent=0.1886, tangentMode=RCTM_User, time=0, value=1}
            {arriveTangent=0.2675, tangentMode=RCTM_User, time=0.966, value=1.1725}
           ]}}}
        Scale Curve=&RandomRangeFloat(Minimum=-1, Maximum=-3)
       )
       1=&FloatFromCurve(
        FloatCurve=DataInterface() {propertyValues {Curve {
           defaultValue=1
           keys [
            {arriveTangent=0.1886, interpMode=RCIM_Cubic, leaveTangent=0.1886, leaveTangentWeight=0.3302, tangentMode=RCTM_User, time=0, value=1}
            <at 0.966 1.1725>
            <at 0.9906 1.75>
           ]
          }}}
        Scale Curve=&RandomRangeFloat(Minimum=1, Maximum=3)
       )
       2=?
       3=?
       Index=$Particles.PartitionParticles.Partition
       Num=::NewEnumerator0
      )
      "Waveform [1]"=::NewEnumerator0
      "[1] Phase"=$Particles.Age
      "[1] Frequency"=1.5
      "[1] Amplitude Min/Max"=<-1 1>
      "[1] Scale"=&RandomRangeFloat(Minimum=10, Maximum=12)
     )
     Y=&Waveform(
      Global Input Position=$Particles.CustomRandom
      Global Amplitude Scale=&MergeHelper_FloatSwitch(
       0=&FloatFromCurve(
        FloatCurve=DataInterface() {propertyValues {Curve {keys [
            {arriveTangent=1.8727, interpMode=RCIM_Cubic, leaveTangent=1.8727, leaveTangentWeight=0.3333, tangentMode=RCTM_User, time=0, value=0}
            <at 0.9906 1.75>
           ]}}}
        Scale Curve=&RandomRangeFloat(Minimum=-1, Maximum=-2)
       )
       1=&FloatFromCurve(
        FloatCurve=DataInterface() {propertyValues {Curve {keys [
            {arriveTangent=1.8727, interpMode=RCIM_Cubic, leaveTangent=1.8727, leaveTangentWeight=0.3333, tangentMode=RCTM_User, time=0, value=0}
            <at 0.9906 1.75>
           ]}}}
        Scale Curve=&RandomRangeFloat(Minimum=-1, Maximum=-2)
       )
       2=?
       3=?
       Index=$Particles.PartitionParticles.Partition
       Num=::NewEnumerator0
      )
      "Waveform [1]"=::NewEnumerator1
      "[1] Phase"=$Particles.Age
      "[1] Frequency"=1.5
      "[1] Amplitude Min/Max"=<-1 1>
      "[1] Scale"=&RandomRangeFloat(Minimum=10, Maximum=5)
     )
     Z=&Add_Float(
      A=0
      B=&FloatFromCurve(
       FloatCurve=DataInterface() {propertyValues {Curve {keys [
           {
            arriveTangent=0.1443
            arriveTangentWeight=0.3614
            interpMode=RCIM_Cubic
            leaveTangent=0.1443
            leaveTangentWeight=0.3614
            tangentMode=RCTM_User
            tangentWeightMode=RCTWM_WeightedBoth
            time=0
            value=0
           }
           {arriveTangent=2.1092, interpMode=RCIM_Cubic, leaveTangent=2.1092, tangentMode=RCTM_User, time=1.0108, value=1.3147}
          ]}}}
       CurveIndex=&Divide_Float(A=$Particles.Age, B=1)
       Scale Curve=&RandomRangeFloat(Minimum=100, Maximum=200)
      )
     )
    )) {isSetParameters=true}
   /SetVariables(Particles.Position=&AddVectorToPosition(
     Position=$Output.SocketLocation.SampledPosition
     Vector=&FN_FrontEndVector3Modifier(
      Gameplay=$Particles.SpiralVector
      Front End=&RotateVectorByQuaternion(
       Vector To Rotate=$Particles.SpiralVector
       Quaternion=&MakeQuatFromAxisAngle(Axis=<1 0 0>, AngleInDegrees=&Lerp_Float(A=20, B=45, Alpha=$System.FN_Contrail_Setup.Dive Value))
      )
     )
    )) {isSetParameters=true}
   &SubUVAnimation(
    Define SubUV Setup Manually:ENiagara_AutomaticManual=::NewEnumerator1
    End Frame Range Override=36
    Manual SubImage Size:Vector2f=<6 6>
    Random Start Frame=true
    Start Frame Offset:NiagaraInt32=&RandomRangeInt(Minimum=-36, Maximum=36)
    SubUV Lookup Index Scale:NiagaraFloat=&SelectFloatFromArray(
     Array Sampling Mode=::NewEnumerator0
     Direct Array Index=?
     Float Selection Array=DataInterface() {propertyValues {FloatData [0.3,0.75,0.5]}}
     Random Seed=0
    )
    UseEndFrame=true
    UseStartFrame=true
   ) {module=SubUVAnimation002}
   &ScaleColor(
    Scale RGB=&VectorFromFloat(Value=&FN_FrontEndFloatModifier(
      Gameplay=&FNTimeOfDayScalar(Night Scale=0.2, Morning Scale=0.25, Day Scale=0.5, Evening Scale=0.35)
      Front End=0.31
     ))
    ScaleA=true
    ScaleRGB=true
   )
   &UpdateMeshOrientation(Rotation Rate=&RandomRangeFloat(Minimum=0.3, Maximum=0.6), Rotation Vector=<0 0 1>)
   &SubUVAnimation(
    Mesh Renderer=DataInterface() {propertyValues {MeshRenderer=&self:Leaves02_Hand_0.NiagaraMeshRendererProperties_1}}
    Renderer Type:ENiagaraMeshOrSprite=::NewEnumerator1
    Start Frame Offset:NiagaraInt32=&RandomRangeInt(Minimum=0, Maximum=32)
    UseEndFrame=false
    UseStartFrame=false
   )
  ]
  rendererClasses [&NiagaraMeshRendererProperties]
  renderers [{rendererClass=&NiagaraMeshRendererProperties, rendererIndex=0}]
 }
 Leaves02_Slow {
  emitterUpdateScript [
   &EmitterState(Delay First Loop Only=true, Life Cycle Mode=::NewEnumerator1, Loop Delay=0.2, Loop Duration=0.25, UseLoopDelay=true)
   &SpawnRate(
    Spawn Probability=0.5
    SpawnRate=&Lerp_Float(A=4, B=10, Alpha=$System.FN_Contrail_Setup.Dive Value)
    Use Spawn Probability=true
   ) {module=SpawnRate001}
  ]
  particleSpawnScript [
   &PartitionParticles()
   &InitializeParticle(
    Color=&SelectLinearColorFromArray(
     Array Sampling Mode=::NewEnumerator0
     Direct Array Index=?
     Color Selection Array=DataInterface() {propertyValues {ColorData [<rgba 0.75 0.275 0.0375 1>, <rgba 0.7188 0.1702 0.0179 1>, <rgba 0.75 0.4712 0.024 1>, <rgba 0.8125 0.1198 0.0264 1>]}}
     Random Seed=0
    )
    Lifetime Min=1.5
    Lifetime Mode=::NewEnumerator1
    Mesh Scale=&Multiply_VectorByFloat(
     Vector=&RandomRangeVector(Minimum=<1 1 1>, Maximum=<1.2 1.2 1.2>)
     Float=&RandomRangeFloat(Minimum=0.085, Maximum=0.12)
    )
    Mesh Scale Mode=::NewEnumerator4
    Sprite Rotation Mode=::NewEnumerator2
    Sprite Size Mode=::NewEnumerator4
    UsePositionOffset=false
    Write Material Random=true
   ) {module=InitializeParticle001}
   &SocketLocation(
    Interpolate Socket Sampling=true
    SamplePosOffset=false
    Socket Index=$Particles.PartitionParticles.Partition
    Socket Reader=$System.Socket Reader
    Socket Sampling Mode=::NewEnumerator3
    VelocitySpeedLimit=false
   ) {module=SocketLocation001}
   &ShapeLocation(
    MaskSpawnGroup=false
    Offset Mode=::NewEnumerator0
    Override Seed=false
    Sphere Radius=&RandomRangeFloat(Minimum=10, Maximum=30, Evaluation Type=::NewEnumerator1)
    UseOwnerScale=true
   ) {module=ShapeLocation001}
   &AddVelocity(
    Velocity=&FN_FrontEndVector3Modifier(
     Gameplay=&RandomRangeVector(Minimum=<0 0 100>, Maximum=<0 0 200>)
     Front End=&Lerp_Vector(A=<0 0 150>, B=<0 -200 200>, Alpha=$System.FN_Contrail_Setup.Dive Value)
    )
    Velocity Speed Scale=&Lerp_Float(A=1, B=4, Alpha=$System.FN_Contrail_Setup.Dive Value)
   ) {module=AddVelocity002}
   &CameraOffset(Camera Offset Amount=&RandomRangeFloat(Minimum=-15, Maximum=-25))
  ]
  particleUpdateScript [
   &ParticleState()
   &ScaleMeshSize(Scale Factor=&VectorFromFloat(Value=&FloatFromCurve(FloatCurve=DataInterface() {propertyValues {Curve {
         defaultValue=1
         keys [
          {arriveTangent=7.4671, interpMode=RCIM_Cubic, leaveTangent=7.4671, tangentMode=RCTM_User, time=0, value=0}
          {arriveTangent=0.0097, arriveTangentWeight=0.0333, interpMode=RCIM_Cubic, leaveTangent=0.0097, tangentMode=RCTM_User, time=0.1, value=1}
          {arriveTangentWeight=0.1333, interpMode=RCIM_Cubic, tangentMode=RCTM_User, time=0.5, value=1}
          <at 1 0>
         ]
        }}})))
   &UpdateMeshOrientation(Rotation Rate=&RandomRangeFloat(Minimum=-0.5, Maximum=0.5), Rotation Vector=<0.3 0.3 1>)
   &SubUVAnimation(
    Define SubUV Setup Manually:ENiagara_AutomaticManual=::NewEnumerator1
    End Frame Range Override=36
    Manual SubImage Size:Vector2f=<6 6>
    Random Start Frame=true
    Start Frame Offset:NiagaraInt32=&RandomRangeInt(Minimum=-36, Maximum=36)
    SubUV Lookup Index Scale:NiagaraFloat=&SelectFloatFromArray(
     Array Sampling Mode=::NewEnumerator0
     Direct Array Index=?
     Float Selection Array=DataInterface() {propertyValues {FloatData [1,0.75,0.5]}}
     Random Seed=0
    )
    UseEndFrame=true
    UseStartFrame=true
   ) {module=SubUVAnimation002}
   &CurlNoiseForce(
    Noise Frequency=300
    Noise Strength=&Multiply_Float(
     A=200
     B=&FloatFromCurve(FloatCurve=DataInterface() {propertyValues {Curve {keys [
          {arriveTangent=4.1302, interpMode=RCIM_Cubic, leaveTangent=4.1302, tangentMode=RCTM_User, time=0.0096, value=0}
          {arriveTangent=0.2664, interpMode=RCIM_Cubic, leaveTangent=0.2664, tangentMode=RCTM_User, time=0.2355, value=1}
         ]}}})
    )
    Pan Noise=false
   )
   &VortexForce(
    UseOriginOffset=false
    Vortex Axis=&RandomRangeVector(Minimum=<0 -1 -1>, Maximum=<0 1 1>)
    Vortex Force Amount=&RandomRangeFloat(Minimum=0, Maximum=0)
   )
   &SolveForcesAndVelocity(Clamp Velocity=false, Limit Acceleration=false)
   &ScaleColor(
    Scale RGB=&VectorFromFloat(Value=&FN_FrontEndFloatModifier(
      Gameplay=&FNTimeOfDayScalar(Night Scale=0.2, Morning Scale=0.25, Day Scale=0.5, Evening Scale=0.35)
      Front End=0.31
     ))
    ScaleA=true
    ScaleRGB=true
   )
   &DynamicMaterialParameters(Module.Index0Float0Write=true, Module.Index0Float1Write=true, Module.Index0Float2Write=true, Module.Index0Float3Write=true)
  ]
  rendererClasses [&NiagaraMeshRendererProperties]
  renderers [{rendererClass=&NiagaraMeshRendererProperties, rendererIndex=0}]
 }
 Spline_Thin {
  emitterUpdateScript [
   &EmitterState()
   &SpawnRate(
    SpawnRate=&FN_FrontEndFloatModifier(
     Gameplay=&Lerp_Float(A=5, B=15, Alpha=$System.FN_Contrail_Setup.Dive Value)
     Front End=&Lerp_Float(A=5, B=15, Alpha=$System.FN_Contrail_Setup.Dive Value)
    )
    Use Spawn Probability=false
   ) {module=SpawnRate001}
  ]
  particleSpawnScript [
   &PartitionParticles()
   &InitializeParticle(
    Color=<rgba 1 0.9279 0.9684 0.025>
    Lifetime=&MergeHelper_FloatSwitch(
     0=&RandomRangeFloat(Minimum=0.6, Maximum=1)
     1=&RandomRangeFloat(Minimum=0.6, Maximum=1)
     2=&RandomRangeFloat(Minimum=0.75, Maximum=1.5)
     3=?
     Index=$Particles.PartitionParticles.Partition
     Num=::NewEnumerator1
    )
    Mesh Scale=&MergeHelper_VectorSwitch(
     0=&RandomRangeVector(Minimum=<1 1 2>, Maximum=<1.5 1.5 5>)
     1=&RandomRangeVector(Minimum=<1 1 2>, Maximum=<1.5 1.5 5>)
     2=&RandomRangeVector(Minimum=<2 2 0.5>, Maximum=<2 2 2>)
     3=?
     Index=$Particles.PartitionParticles.Partition
     Num=::NewEnumerator1
    )
    Mesh Scale Mode=::NewEnumerator4
    Position Mode=::NewEnumerator0
    Write Material Random=true
   ) {module=InitializeParticle001}
   &SocketLocation(
    Interpolate Socket Sampling=true
    Orientation Sampling=::NewEnumerator1
    SamplePosOffset=false
    Socket Index=$Particles.PartitionParticles.Partition
    Socket Reader=$System.Socket Reader
    Socket Sampling Mode=::NewEnumerator3
    VelocitySpeedLimit=false
   ) {module=SocketLocation001}
   &InitialMeshOrientation(
    Mesh Axis To Orient=&FN_FrontEndVector3Modifier(Gameplay=<0 0 1>, Front End=&Lerp_Vector(A=<0 0 1>, B=<0 1 1>, Alpha=$System.FN_Contrail_Setup.Dive Value))
    Mesh Orientation Mode=::NewEnumerator2
    Orientation Vector=&MergeHelper_VectorSwitch(0=<0 0 -1>, 1=<0 0 -1>, 2=?, 3=?, Index=$Particles.PartitionParticles.Partition, Num=::NewEnumerator0)
    Rotation=&RandomRangeVector(Minimum=<0 0 0>, Maximum=<0 0 1>)
    Use Rotation Vector=true
   ) {module=InitialMeshOrientation001}
   &CameraOffset(Camera Offset Amount=5)
  ]
  particleUpdateScript [
   &ParticleState()
   &DynamicMaterialParameters(
    Index 0 Param 1=&FloatFromCurve(
     FloatCurve=DataInterface() {propertyValues {Curve {keys [
         {interpMode=RCIM_Cubic, tangentMode=RCTM_User, time=0, value=1}
         {arriveTangent=-3.0019, interpMode=RCIM_Cubic, leaveTangent=-3.0019, tangentMode=RCTM_User, time=1, value=0}
        ]}}}
     Scale Curve=&RandomRangeFloat(Minimum=5, Maximum=10)
    )
    Index 0 Param 2=&RandomRangeFloat(Minimum=-5, Maximum=-3)
    Index 0 Param 3=&FloatFromCurve(FloatCurve=DataInterface() {propertyValues {Curve {keys [<at 0 1 RCIM_Cubic>, {arriveTangent=-0.8028, interpMode=RCIM_Cubic, leaveTangent=-0.8028, tangentMode=RCTM_User, time=1, value=0}]}}})
    Index 0 Param 4=&MaskFloatBySpawnGroup(Masked Float=0, Passthrough Float=0.5)
    Module.Param0WriteEnabled=true
    Module.Param1WriteEnabled=true
    Module.Param2WriteEnabled=true
    Module.Param3WriteEnabled=true
   )
   &UpdateMeshOrientation(Rotation Rate=&RandomRangeFloat(Minimum=-0.5, Maximum=0.5), Rotation Vector=<0 0 1>)
   &ScaleMeshSize(Scale Factor=&VectorFromCurve(VectorCurve=DataInterface() {propertyValues {
       XCurve {keys [
         {arriveTangent=2.2128, interpMode=RCIM_Cubic, leaveTangent=2.2128, tangentMode=RCTM_User, time=0, value=1}
         {arriveTangent=-0.9, interpMode=RCIM_Cubic, leaveTangent=-0.9, time=0.2, value=1.4}
         {arriveTangent=-2.7027, interpMode=RCIM_Cubic, leaveTangent=-2.7027, tangentMode=RCTM_User, time=1, value=0.1}
        ]}
       YCurve {keys [
         {arriveTangent=2.2128, interpMode=RCIM_Cubic, leaveTangent=2.2128, tangentMode=RCTM_User, time=0, value=1}
         {arriveTangent=-0.9, interpMode=RCIM_Cubic, leaveTangent=-0.9, time=0.2, value=1.4}
         {arriveTangent=-2.7027, interpMode=RCIM_Cubic, leaveTangent=-2.7027, tangentMode=RCTM_User, time=1, value=0.1}
        ]}
       ZCurve {keys [<at 0 0.4>,<at 1 1>]}
      }}))
   &ScaleColor(
    Scale Alpha=&FloatFromCurve(
     FloatCurve=DataInterface() {propertyValues {Curve {keys [<at 0 1 RCIM_Cubic>, {arriveTangent=-2.2348, interpMode=RCIM_Cubic, leaveTangent=-2.2348, tangentMode=RCTM_User, time=1, value=0}]}}}
     Scale Curve=&FN_FrontEndFloatModifier(Gameplay=1, Front End=0.2)
    )
    Scale RGB=&FN_FrontEndVector3Modifier(Gameplay=<1 1 1>, Front End=<4 2 2>)
    ScaleA=true
    ScaleRGB=true
   )
  ]
  rendererClasses [&NiagaraMeshRendererProperties]
  renderers [{rendererClass=&NiagaraMeshRendererProperties, rendererIndex=0}]
 }
 Trails {
  emitterUpdateScript [
   &EmitterState(
    Inactive Response=::NewEnumerator1
    Life Cycle Mode=::NewEnumerator1
    Loop Behavior=::NewEnumerator1
    Loop Duration Mode=::NewEnumerator1
    UseLoopDelay=false
   )
   &SpawnBurst_Instantaneous(Spawn Count=2, Use Spawn Probability=false)
  ]
  particleSpawnScript [
   &FN_SetSpawnGroup(Number of Spawn Groups=2)
   &InitializeParticle(
    Color=<rgba 1 0.3802 0.4491 0.135>
    Mesh Scale=<0.8 0.8 3>
    Mesh Scale Mode=::NewEnumerator4
    UsePositionOffset=false
    Write Material Random=true
   )
   &DynamicMaterialParameters(
    Index 0 Param 1=&RandomRangeFloat(Minimum=0.2, Maximum=0.4)
    Index 0 Param 2=&RandomRangeFloat(Minimum=0, Maximum=10)
    Index 0 Param 3=30
    Index 0 Param 4=1
    Module.Param0WriteEnabled=true
    Module.Param1WriteEnabled=true
    Module.Param2WriteEnabled=false
    Module.Param3WriteEnabled=true
   )
   /SetVariables(Particles.RotatorSpeedRandom:NiagaraFloat=&RandomRangeFloat(Minimum=0.3, Maximum=-0.3)) {isSetParameters=true}
   &CameraOffset(Camera Offset Amount=10)
   &InitialMeshOrientation(
    Orientation Axis=<0 0 1>
    Orientation Vector=<0 0 2.751>
    Rotation=&RandomRangeVector(Minimum=<0 0 0>, Maximum=<0 0 1>)
    Use Orientation Vector=true
    Use Rotation Vector=true
   )
  ]
  particleUpdateScript [
   &ParticleState(Kill Particles When Lifetime Has Elapsed=false, Let Infinitely Lived Particles Die When Emitter Deactivates=true)
   &InterpolateOverTime(Rate Of Change=10, Target Vector Value:NiagaraPosition=$Particles.Position, Variable Type=::NewEnumerator2)
   /SetVariables(Particles.InterpolateDynamic:Vector4f=&MakeFloat4FromVectorAndFloat("Vector (RGB)"=$Output.InterpolateOverTime.Moving Average, "Float (Alpha)"=1)) {isSetParameters=true}
   &SkeletalMeshLocation(
    Bone / Socket Index=$Particles.SpawnGroup
    Bone Sampling Mode=::NewEnumerator4
    Inherit Velocity=false
    Interpolate Mesh Sampling=false
    MaskSpawnGroup=false
    Skeletal Mesh=$System.Skeletal Mesh
   )
   &SolveForcesAndVelocity(Clamp Velocity=false, Limit Acceleration=false)
   &DynamicMaterialParameters(
    Index 0 Param 3=&Multiply_Float(A=30, B=&Lerp_Float(A=0.5, B=1, Alpha=$System.Dive))
    Index 2 Param 1=&Lerp_Float(A=0, B=1.5, Alpha=&Ceil_Float(Float=$System.Dive))
    Index 2 Param 2=&Lerp_Float(A=0, B=0.5, Alpha=&Ceil_Float(Float=$System.Dive))
    Module.Param0WriteEnabled=false
    Module.Param0WriteEnabled2=true
    Module.Param1WriteEnabled=false
    Module.Param1WriteEnabled2=true
    Module.Param2WriteEnabled=true
    Module.Param2WriteEnabled2=false
    Module.Param3WriteEnabled=false
    Module.Param3WriteEnabled2=false
    Write Parameter Index 2=true
   ) {module=DynamicMaterialParameters001}
   &ScaleColor(
    Scale Alpha=&MakeCustomFloatFromBool(
     Bool=$User.bIsFrontEndPreview
     True Float=&Lerp_Float(A=0.15, B=0.3, Alpha=$System.Dive)
     False Float=&Multiply_Float(A=&Lerp_Float(A=0.5, B=1, Alpha=$System.Dive), B=0.8)
    )
    Scale RGB=&FN_FrontEndVector3Modifier(
     Gameplay=&VectorFromFloat(Value=&FN_Select_Float_from_Time_Of_Day(Morning=2, Day=0.4, Evening=0.5, Night=0.35))
     Front End=<0.5 0.5 0.5>
    )
    ScaleA=true
    ScaleRGB=true
   )
   &ScaleMeshSize(Scale Factor=&MakeVector(
     X=&Lerp_Float(A=1, B=2, Alpha=$System.FN_Contrail_Setup.Dive Value)
     Y=&Lerp_Float(A=1, B=2, Alpha=$System.FN_Contrail_Setup.Dive Value)
     Z=1
    )) {module=ScaleMeshSize001}
   /SetVariables(Particles.MeshOrientation_FE:Quat4f=&LerpQuaternion(
     Quaternion A:Quat4f=&MakeQuatFromAxisAngle(
      Axis=&TransformVector(Vector=<1 0 0>, Destination Space=::World)
      AngleInDegrees=&FN_FrontEndFloatModifier(Gameplay=20, Front End=&Lerp_Float(A=0, B=45, Alpha=$System.FN_Contrail_Setup.Dive Value))
     )
     Quaternion B:Quat4f=&MakeQuatFromAxisAngle(Axis=&TransformVector(Vector=<1 0 0>, Destination Space=::World), AngleInDegrees=45)
     Lerp Factor:NiagaraFloat=$System.Dive
     Enable Spherical Linear Interpolation:NiagaraBool=false
    )) {isSetParameters=true}
   /SetVariables(Particles.NewMeshOrientation:Quat4f=/Niagara/DynamicInputs/TypeConversions/MakeCustomQuaternionFromBool(Bool:NiagaraBool=$User.bIsFrontEndPreview, True Quat:Quat4f=$Particles.MeshOrientation_FE, False Quat:Quat4f=$Particles.MeshOrientation)) {isSetParameters=true}
  ]
  rendererClasses [&NiagaraMeshRendererProperties]
  renderers [{rendererClass=&NiagaraMeshRendererProperties, rendererIndex=0}]
 }
}
summary {
 systemName=NS_Contrail_SpeedGreen
 emitters [
  {emitterName=Trails, rendererClasses [&NiagaraMeshRendererProperties]}
  {emitterName=Leaves01, rendererClasses [&NiagaraSpriteRendererProperties]}
  {emitterName=Leaves02_Slow, rendererClasses [&NiagaraMeshRendererProperties]}
  {emitterName=Leaves02_Hand, rendererClasses [&NiagaraMeshRendererProperties]}
  {emitterName=Spline_Thin, rendererClasses [&NiagaraMeshRendererProperties]}
 ]
 userVariables [
  User.bIsFrontEnd {type=NiagaraBool}
  User.bIsFrontEndPreview {type=NiagaraBool}
  User.ParaGlideLeanAlpha {type=NiagaraFloat, defaultValue=0, description=""}
  User.RibbonIDOffset {type=NiagaraInt32}
 ]
}
